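# Unlit Niagara sprite and mesh particle material for the Veiled backpack effect, producing shifting iridescent color.
# Scaled UVs, shifted by a dynamic-parameter Offset, sample blurred noise, whose amplified red channel becomes the lookup coordinate into an iridescent gradient.
# The gradient gets a constant tint added, then passes through Desaturation, brightness and particle color into the shared emissive-adjustment function.
# A saturated camera depth fade is also computed.
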Material /Game/Accessories/FORT_Backpacks/Backpack_F_MED_Veiled/FX/M_Backpack_Veiled_Square
graph {
 properties {
  bUsedWithNiagaraMeshParticles=true
  bUsedWithNiagaraSprites=true
  pixelDepthOffsetMode=PDOM_Legacy
  shadingModel=MSM_Unlit
  usedShadingModels=Unlit
 }
 nodes [
  n1 = Multiply(A=TextureCoordinate(), B=ScalarParameter() {defaultValue=0.03, parameterName=Scale})
  n2 = DynamicParameter() {paramNames [Offset,Param2,Param3,Param4]}
  n3 = Multiply(A=TextureSample(UVs=Add(A=@n1, B=@n2.Offset)) {samplerType=SAMPLERTYPE_Grayscale, texture=/FortUILibrary/UIKit/Buttons/Textures/LowResBlurredNoise}.R) {constB=4}
  n4 = Desaturation(
   None=Add(
    A=Constant4Vector() {constant=<rgba 0.5 0.6 0.4 0>}.RGBA
    B=TextureSample(UVs=@n3) {texture=/Game/Weapons/WeaponSkins/Wraps/Textures/Iridescence/T_iridescent_gradient}.RGBA
   )
   Fraction=ScalarParameter() {defaultValue=7, parameterName=Saturation}
  )
  n5 = ScalarParameter() {defaultValue=0.8, parameterName=Brightness}
  MaterialFunctionCall(Emissive=Multiply(A=ParticleColor().RGB, B=Multiply(A=@n4, B=@n5))) {materialFunction=&MF_FX_EmissiveAdjustments}
  Saturate(None=MaterialFunctionCall(Fade Length=Constant() {r=220}, Fade Offset=Constant() {r=80}) {materialFunction=&CameraDepthFade}.Result)
  ScalarParameter() {parameterName=Param_2}
 ]
}
